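; Rift beacon quest prop Blueprint that poses its reflector, four legs and feet from per-instance angle variables.
; When bRunConstructionScript is set, the construction script calls Run Init Logic to apply every pose step.
; The reflector angle is clamped, then applied as a relative transform.
; Front legs rotate in place, rear legs snap to body sockets with yaw offsets, and feet attach at leg socket positions.
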(event Custom|Run%20Init%20Logic
  (Default|SetBeaconPosition)
  (Default|SetLegRotationFrontRight)
  (Default|SetLegRotationFrontLeft)
  (Default|SetFootPositionFrontRight)
  (Default|SetFootPositionFrontLeft)
  (Default|SetLegRotationRearRight)
  (Default|SetLegRotationRearLeft)
  (Default|SetFootPositionRearRight)
  (Default|SetFootPositionRearLeft))

(fn SetBeaconPosition
  (bind _beaconangle (Variables|InitSpawnValues|GetBeaconAngle))
  (if (< _beaconangle -145.0)
    (Variables|InitSpawnValues|SetBeaconAngle -145.0))
  (if (> _beaconangle 2e+01)
    (Variables|InitSpawnValues|SetBeaconAngle 2e+01)
    (else
      _))
  (Transformation|SetRelativeTransform (Variables|Default|GetSM_Quest_RiftBeaconBReflector_A) (Math|Transform|MakeTransform "0.000000,24.000000,147.000000" (Math|Rotator|MakeRotator (Variables|InitSpawnValues|GetBeaconAngle))) false false))

(fn SetFootPosition_FrontLeft
  (bind (_returnvalue_location _returnvalue_rotation _returnvalue_scale) (Math|Transform|BreakTransform (Transformation|GetActorTransform self)))
  (Transformation|SetWorldTransform (Variables|Default|GetFrontFoot_Left) (Math|Transform|MakeTransform (Transformation|GetSocketLocation (Variables|Default|GetFrontLeg_Left) (Utilities|Array|Get(aref) (Transformation|GetAllSocketNames (Variables|Default|GetFrontLeg_Left)))) (Math|Rotator|MakeRotator (Variables|InitSpawnValues|GetFrontFootLeftAngle) (.pitch (Transformation|GetWorldRotation (Variables|Default|GetFrontLeg_Left))) (.yaw (Transformation|GetWorldRotation (Variables|Default|GetFrontLeg_Left)))) _returnvalue_scale) false false))

(fn SetFootPosition_FrontRight
  (bind (_returnvalue_location _returnvalue_rotation _returnvalue_scale) (Math|Transform|BreakTransform (Transformation|GetActorTransform self)))
  (Transformation|SetWorldTransform (Variables|Default|GetFrontFoot_Right) (Math|Transform|MakeTransform (Transformation|GetSocketLocation (Variables|Default|GetFrontLeg_Right) (Utilities|Array|Get(aref) (Transformation|GetAllSocketNames (Variables|Default|GetFrontLeg_Right)))) (Math|Rotator|MakeRotator (Variables|InitSpawnValues|GetFrontFootRightAngle) (.pitch (Transformation|GetWorldRotation (Variables|Default|GetFrontLeg_Right))) (.yaw (Transformation|GetWorldRotation (Variables|Default|GetFrontLeg_Right)))) _returnvalue_scale) false false))

(fn SetFootPosition_RearLeft
  (bind (_returnvalue_location _returnvalue_rotation _returnvalue_scale) (Math|Transform|BreakTransform (Transformation|GetActorTransform self)))
  (bind (_returnvalue_roll _returnvalue_pitch _returnvalue_yaw) (Math|Rotator|BreakRotator (Transformation|GetWorldRotation (Variables|Debug|GetStaticMeshComponent))))
  (Transformation|SetWorldTransform (Variables|Default|GetRearFoot_Left) (Math|Transform|MakeTransform (Transformation|GetSocketLocation (Variables|Default|GetRearLeg_Left) (Utilities|Array|Get(aref) (Transformation|GetAllSocketNames (Variables|Default|GetRearLeg_Left)))) (Math|Rotator|MakeRotator (Variables|InitSpawnValues|GetRearFootLeftAngle) 0.0 (+ _returnvalue_yaw -2.1e+02)) _returnvalue_scale) false false))

(fn SetFootPosition_RearRight
  (bind (_returnvalue_location _returnvalue_rotation _returnvalue_scale) (Math|Transform|BreakTransform (Transformation|GetActorTransform self)))
  (bind (_returnvalue_roll _returnvalue_pitch _returnvalue_yaw) (Math|Rotator|BreakRotator (Transformation|GetWorldRotation (Variables|Debug|GetStaticMeshComponent))))
  (Transformation|SetWorldTransform (Variables|Default|GetRearFoot_Right) (Math|Transform|MakeTransform (Transformation|GetSocketLocation (Variables|Default|GetRearLeg_Right) (Utilities|Array|Get(aref) (Transformation|GetAllSocketNames (Variables|Default|GetRearLeg_Right)))) (Math|Rotator|MakeRotator (Variables|InitSpawnValues|GetRearFootRightAngle) 0.0 (+ _returnvalue_yaw 2.1e+02)) _returnvalue_scale) false false))

(fn SetLegRotation_FrontLeft
  (Transformation|SetWorldTransform (Variables|Default|GetFrontLeg_Left) (Math|Transform|MakeTransform (.location (Transformation|GetWorldTransform (Variables|Default|GetFrontLeg_Left))) (Math|Rotator|MakeRotator (Variables|InitSpawnValues|GetFrontLegLeftAngle) (.pitch (.rotation (Transformation|GetWorldTransform (Variables|Default|GetFrontLeg_Left)))) (.yaw (.rotation (Transformation|GetWorldTransform (Variables|Default|GetFrontLeg_Left))))) (.scale (Transformation|GetWorldTransform (Variables|Default|GetFrontLeg_Left)))) false false))

(fn SetLegRotation_FrontRight
  (Transformation|SetWorldTransform (Variables|Default|GetFrontLeg_Right) (Math|Transform|MakeTransform (.location (Transformation|GetWorldTransform (Variables|Default|GetFrontLeg_Right))) (Math|Rotator|MakeRotator (Variables|InitSpawnValues|GetFrontLegRightAngle) (.pitch (.rotation (Transformation|GetWorldTransform (Variables|Default|GetFrontLeg_Right)))) (.yaw (.rotation (Transformation|GetWorldTransform (Variables|Default|GetFrontLeg_Right))))) (.scale (Transformation|GetWorldTransform (Variables|Default|GetFrontLeg_Right)))) false false))

(fn SetLegRotation_RearLeft
  (bind (_returnvalue_location _returnvalue_rotation _returnvalue_scale) (Math|Transform|BreakTransform (Transformation|GetActorTransform self)))
  (bind (_returnvalue_roll _returnvalue_pitch _returnvalue_yaw) (Math|Rotator|BreakRotator (Transformation|GetWorldRotation (Variables|Debug|GetStaticMeshComponent))))
  (Transformation|SetWorldTransform (Variables|Default|GetRearLeg_Left) (Math|Transform|MakeTransform (Transformation|GetSocketLocation (Variables|Debug|GetStaticMeshComponent) (Utilities|Array|Get(aref) (Transformation|GetAllSocketNames (Variables|Debug|GetStaticMeshComponent)) 1)) (Math|Rotator|MakeRotator (Variables|InitSpawnValues|GetRearLegLeftAngle) 0.0 (+ _returnvalue_yaw -3e+01)) _returnvalue_scale) false false))

(fn SetLegRotation_RearRight
  (bind (_returnvalue_location _returnvalue_rotation _returnvalue_scale) (Math|Transform|BreakTransform (Transformation|GetActorTransform self)))
  (bind (_returnvalue_roll _returnvalue_pitch _returnvalue_yaw) (Math|Rotator|BreakRotator (Transformation|GetWorldRotation (Variables|Debug|GetStaticMeshComponent))))
  (Transformation|SetWorldTransform (Variables|Default|GetRearLeg_Right) (Math|Transform|MakeTransform (Transformation|GetSocketLocation (Variables|Debug|GetStaticMeshComponent) (Utilities|Array|Get(aref) (Transformation|GetAllSocketNames (Variables|Debug|GetStaticMeshComponent)))) (Math|Rotator|MakeRotator (Variables|InitSpawnValues|GetRearLegRightAngle) 0.0 (+ _returnvalue_yaw 3e+01)) _returnvalue_scale) false false))

(fn ConstructionScript
  (if (Variables|Default|GetbRunConstructionScript)
    (CallFunction|RunInitLogic)))
